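# Niagara system for the glow effect of the OathBound main-player emote.
# Its single MainGlow emitter burst-spawns particles with a fixed lifetime, a warm HDR color and a material random value, drawn by two mesh renderers.
# Skeletal Mesh Location runs in particle update rather than spawn, re-sampling position from a system-level skeletal mesh reader so the glow tracks the animated mesh.
# The system exposes no user parameters.
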NiagaraSystem /BRCosmetics/Animation/Game/MainPlayer/Emotes/OathBound/FX/NS_Emote_OathBound_Glow
dependencies {
 usedDataInterfaces [&NiagaraDataInterfaceSkeletalMesh]
 usedDynamicInputs []
 usedModules [
  &self:SystemSpawnScript.SystemScriptSource.SystemScriptGraph.NiagaraNodeAssignment_1.SetVariables_#1
  &SystemState
 ]
}
emitters {MainGlow {
  emitterUpdateScript [
   &EmitterState()
   &SpawnBurst_Instantaneous(Use Spawn Probability=false, UseLoopCountLimit=false)
  ]
  particleSpawnScript [&InitializeParticle(Color=<rgba 5 3.145 0.599 1>, Lifetime=5, UsePositionOffset=false, Write Material Random=true)]
  particleUpdateScript [
   &ParticleState()
   &SkeletalMeshLocation(Bone Sampling Mode=::NewEnumerator0, MaskSpawnGroup=false, OverrideRandomSeed=false, Skeletal Mesh=$System.Skeletal Mesh Reader)
  ]
  rendererClasses [&NiagaraMeshRendererProperties]
  renderers [
   {rendererClass=&NiagaraMeshRendererProperties, rendererIndex=0}
   {rendererClass=&NiagaraMeshRendererProperties, rendererIndex=1}
  ]
 }}
summary {
 systemName=NS_Emote_OathBound_Glow
 emitters [{emitterName=MainGlow, rendererClasses [&NiagaraMeshRendererProperties]}]
 userVariables []
}
