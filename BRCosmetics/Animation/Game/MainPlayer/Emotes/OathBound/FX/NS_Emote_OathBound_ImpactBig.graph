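# Niagara burst effect for the large impact in the OathBound emote, a brief warm orange flash with flying debris light.
# Flash_Elements packs stars, spikes, flare, glows and a light into one emitter, using per-layer attributes and age curves to feed many renderers.
# Sparks fall with drag and gravity, switching orange or blue by system variant, while Streaks and Minimal add fast radial lines.
# Burst count varies by platform; some alphas scale with time of day.
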NiagaraSystem /BRCosmetics/Animation/Game/MainPlayer/Emotes/OathBound/FX/NS_Emote_OathBound_ImpactBig
dependencies {
 usedDataInterfaces [
  &NiagaraDataInterfacePlatformSet
  &NiagaraDataInterfaceCurve
  &NiagaraDataInterfaceCamera
  &NiagaraDataInterfaceColorCurve
  &NiagaraDataInterfaceVector2DCurve
 ]
 usedDynamicInputs [&MergeHelper_FloatSwitch]
 usedModules [&SystemState]
}
emitters {
 Flash_Elements {
  emitterUpdateScript [
   &EmitterState()
   &SpawnBurst_Instantaneous(Spawn Group=1, Use Spawn Probability=false, UseLoopCountLimit=false) {module=SpawnBurst_Instantaneous001}
   &SpawnBurst_Instantaneous(
    Spawn Count=&MakeIntFromFloat(Float=&FN_PickFloatDependOnPlatform(
      TargetPlatform=DataInterface() {propertyValues {Platforms {qualityLevelMask=-29}}}
      TargetPlatformActive=&MakeFloatFromInt(Int=&RandomRangeInt(Minimum=3, Maximum=5))
      TargetPlatformNoActive=&MakeFloatFromInt(Int=&RandomRangeInt(Minimum=7, Maximum=10))
     ))
    Use Spawn Probability=false
    UseLoopCountLimit=false
   )
  ]
  particleSpawnScript [
   &InitializeParticle(Lifetime=0.56, UsePositionOffset=false, Write Material Random=true)
   &ShapeLocation(MaskSpawnGroup=false, Override Seed=false, Sphere Radius=1, UseOwnerScale=true)
   &SpriteFacingAndAlignment(Sprite Alignment=$Particles.ShapeLocation.ShapeNormal, WriteAlignment=true, WriteFacing=false) {enabled=false}
   &FN_PerParticleLOD(Enable SpriteSize Distance Scaling=true, Sprite Size Scale Far=<2.5 2.5>) {enabled=false}
   /SetVariables(Particles.SpriteSize=<1 1>, Particles.VisibilityTag=&SpawnGroup()) {isSetParameters=true}
   /SetVariables(
    Particles.Flash Spike Alignment:Vector3f=$Particles.ShapeLocation.ShapeNormal
    Particles.Flash Spike Color:LinearColor=<rgba 0.2 0.0833 0 0.8>
    Particles.Flash Spike Lifetime:NiagaraFloat=&RandomRangeFloat(Minimum=0.05, Maximum=0.08)
    Particles.Flash Spike Sprite Size:Vector2f=&RandomRangeVector2D(Minimum=<6 40>, Maximum=<12 50>)
   ) {isSetParameters=true}
   /SetVariables(
    Particles.Flash Star Cam Offset=15
    Particles.Flash Star Lifetime=0.15
    Particles.Flash Star Rotation:NiagaraFloat=&RandomRangeFloat(Minimum=0, Maximum=360)
    Particles.Flash Star Sprite Size:Vector2f=&Multiply_Vector2D_ByFloat(Float=&RandomRangeFloat(Minimum=90, Maximum=160), Vector2D=$Particles.SpriteSize)
    Particles.Flash Star Sprite Size Alt:Vector2f=&Multiply_Vector2D_ByFloat(
     Float=&Multiply_Float(A=0.85, B=&RandomRangeFloat(Minimum=90, Maximum=160))
     Vector2D=$Particles.SpriteSize
    )
    Particles.Flash Star SubUV Index:NiagaraFloat=&RandomRangeFloat(Minimum=0, Maximum=3)
    Particles.Flash Star SubUV Index Alt:NiagaraFloat=&RandomRangeFloat(Minimum=0, Maximum=3)
   ) {isSetParameters=true}
   /SetVariables(
    Particles.Flash Flare Color:LinearColor=<rgba 1 1 1 0.7>
    Particles.Flash Flare Lifetime=0.2
    Particles.Flash Flare Sprite Size:Vector2f=<150 150>
   ) {isSetParameters=true}
   /SetVariables(Particles.Dark_Glow_CameraOffset=50) {isSetParameters=true}
   /SetVariables(Particles.LightLifetime=0.3) {isSetParameters=true}
   &InitialMeshOrientation(Mesh Orientation Mode=::NewEnumerator7, Rotation=<0 0.25 0>, Use Rotation Vector=true)
   &AddVelocity(Velocity=&FN_FrontEndVector3Modifier(
     Gameplay=<10 0 0>
     Front End=&Multiply_Vector(A=&ReturnCameraVector(Camera Vectors=::NewEnumerator1, Coordinate Space=::Local), B=<-1 -1 -1>)
    ))
   &ApplyInitialForces()
  ]
  particleUpdateScript [
   &ParticleState()
   /SetVariables(
    Particles.Update Flash Flare Age:NiagaraFloat=&Divide_Float(A=$Particles.Age, B=$Particles.Initial.Flash Flare Lifetime)
    Particles.Update Flash Spike Age:NiagaraFloat=&Divide_Float(A=$Particles.Age, B=$Particles.Initial.Flash Spike Lifetime)
    Particles.Update Flash Star Age:NiagaraFloat=&Divide_Float(A=$Particles.Age, B=$Particles.Initial.Flash Star Lifetime)
    Particles.Update Light Age:NiagaraFloat=&Divide_Float(A=$Particles.Age, B=$Particles.Initial.LightLifetime)
   ) {isSetParameters=true}
   /SetVariables(
    Particles.Update Flash Spike Color:LinearColor=&MakeLinearColorFromVectorAndFloat(
     "Vector (RGB)"=&MakeVectorFromLinearColorRGB(Color=&MultiplyLinearColor(
       Linear Color=$Particles.Initial.Flash Spike Color
       Scale Factor=&MakeVector4FromVectorAndFloat("Vector (RGB)"=<1 1 1>, "Float (Alpha)"=1)
      ))
     "Float (Alpha)"=&FloatFromCurve(FloatCurve=DataInterface() {propertyValues {Curve {keys [<at 0.7 1>,<at 1 0>]}}}, CurveIndex=$Particles.Update Flash Spike Age)
    )
    Particles.Update Flash Spike Sprite Size:Vector2f=&Multiply_Vector2D_ByFloat(
     Float=&FloatFromCurve(FloatCurve=DataInterface() {propertyValues {Curve {keys [<at 0 0.6>,<at 1 2>]}}}, CurveIndex=$Particles.Update Flash Spike Age)
     Vector2D=$Particles.Initial.Flash Spike Sprite Size
    )
    Particles.Update Flash Star Color:LinearColor=&MultiplyLinearColor(
     Linear Color=&ColorFromCurve(
      ColorCurve=DataInterface() {propertyValues {
        AlphaCurve {keys [<at 0 2>,<at 0.8992 0>]}
        BlueCurve {keys [<at 0 0.0975>,<at 0.3739 0.065>]}
        GreenCurve {keys [<at 0 1.1494>,<at 0.3739 0.4546>]}
        RedCurve {keys [<at 0 1.5>,<at 0.3739 1>]}
       }}
      CurveIndex=$Particles.Update Flash Star Age
     )
     Scale Factor=&MakeVector4FromVectorAndFloat("Vector (RGB)"=<1 1 1>, "Float (Alpha)"=1)
    )
    Particles.Update Flash Star Color Alt:LinearColor=&MultiplyLinearColor(
     Linear Color=&ColorFromCurve(
      ColorCurve=DataInterface() {propertyValues {
        AlphaCurve {keys [<at 0 2>,<at 0.8992 0>]}
        BlueCurve {keys [<at 0.1561 0.0413>,<at 0.8536 0.0278>]}
        GreenCurve {keys [<at 0.1561 0.2299>,<at 0.8536 0.0883>]}
        RedCurve {keys [<at 0.1561 0.6615>,<at 0.8536 0.2031>]}
       }}
      CurveIndex=$Particles.Update Flash Star Age
     )
     Scale Factor=&MakeVector4FromVectorAndFloat("Vector (RGB)"=<1 1 1>, "Float (Alpha)"=1)
    )
    Particles.Update Flash Star Sprize Size:Vector2f=&Multiply_Vector2D_ByFloat(
     Float=&FloatFromCurve(
      FloatCurve=DataInterface() {propertyValues {Curve {keys [
          {arriveTangent=2.7013, interpMode=RCIM_Cubic, leaveTangent=2.7013, tangentMode=RCTM_User, time=0, value=0}
          {interpMode=RCIM_Cubic, tangentMode=RCTM_User, time=1, value=1}
         ]}}}
      CurveIndex=$Particles.Update Flash Star Age
     )
     Vector2D=$Particles.Initial.Flash Star Sprite Size
    )
    Particles.Update Flash Star Sprize Size Alt:Vector2f=&Multiply_Vector2D_ByFloat(
     Float=&FloatFromCurve(
      FloatCurve=DataInterface() {propertyValues {Curve {keys [
          {arriveTangent=2.7013, interpMode=RCIM_Cubic, leaveTangent=2.7013, tangentMode=RCTM_User, time=0, value=0}
          {interpMode=RCIM_Cubic, tangentMode=RCTM_User, time=1, value=1}
         ]}}}
      CurveIndex=$Particles.Update Flash Star Age
     )
     Vector2D=$Particles.Initial.Flash Star Sprite Size Alt
    )
   ) {isSetParameters=true}
   /SetVariables(Particles.Update Flash Flare Color:LinearColor=&MakeLinearColorFromVectorAndFloat(
     "Vector (RGB)"=&Multiply_Vector(
      A=&MakeVectorFromLinearColorRGB(Color=$Particles.Initial.Flash Flare Color)
      B=&MakeVectorFromLinearColorRGB(Color=&MultiplyLinearColor(
        Linear Color=&ColorFromCurve(
         ColorCurve=DataInterface() {propertyValues {
           AlphaCurve {keys [<at 0 1 RCIM_Cubic>,<at 1 1 RCIM_Cubic>]}
           BlueCurve {keys [<at 0 0.5826 RCIM_Cubic>,<at 0.713 0 RCIM_Cubic>]}
           GreenCurve {keys [<at 0 2.1942 RCIM_Cubic>,<at 0.713 0.8333 RCIM_Cubic>]}
           RedCurve {keys [<at 0 3 RCIM_Cubic>,<at 0.713 2 RCIM_Cubic>]}
          }}
         CurveIndex=$Particles.Update Flash Flare Age
        )
        Scale Factor=&MakeVector4FromVectorAndFloat("Vector (RGB)"=<1 1 1>, "Float (Alpha)"=1)
       ))
     )
     "Float (Alpha)"=&FloatFromCurve(FloatCurve=DataInterface() {propertyValues {Curve {keys [<at 0 1>,<at 1 0>]}}}, CurveIndex=$Particles.Update Flash Flare Age)
    )) {isSetParameters=true}
   &FN_Light_Attributes(
    Color=<rgba 3 2.415 0.3 1>
    Create Color Attribute=true
    Create Exponent Attribute=true
    Create Position Attribute=true
    Create Radius Attribute=true
    Exponent=7
    Position=<15 0 0>
    Radius=90
    Scale Color=&FloatFromCurve(
     FloatCurve=DataInterface() {propertyValues {Curve {keys [
         {interpMode=RCIM_Cubic, tangentMode=RCTM_User, time=0, value=1}
         {arriveTangent=-2.0035, interpMode=RCIM_Cubic, leaveTangent=-2.0035, tangentMode=RCTM_User, time=1, value=0}
        ]}}}
     CurveIndex=$Particles.Update Light Age
     Scale Curve=10
    )
   )
   &FN_Sprite_Attributes(
    Color=&ColorFromCurve(ColorCurve=DataInterface() {propertyValues {
       AlphaCurve {keys [<at 0 2.5 RCIM_Cubic>,<at 1 0 RCIM_Cubic>]}
       BlueCurve {keys [<at 0 0.1243>,<at 0.2466 0.0497>]}
       GreenCurve {keys [<at 0 0.7009>,<at 0.2466 0.2804>]}
       RedCurve {keys [<at 0 2>,<at 0.2466 0.8>]}
      }})
    Create Color Attribute=true
    Create SpriteSize Attribute=true
    SpriteSize=&Multiply_Vector2D(
     A=<157.0522 163.3067>
     B=&Vector2DFromFloat(Value=&FloatFromCurve(FloatCurve=DataInterface() {propertyValues {Curve {keys [
           {arriveTangent=2.7817, interpMode=RCIM_Cubic, leaveTangent=2.7817, tangentMode=RCTM_User, time=0, value=0.375}
           {arriveTangent=-0.4253, interpMode=RCIM_Cubic, leaveTangent=-0.4253, tangentMode=RCTM_User, time=0.1982, value=1.3517}
           {tangentMode=RCTM_User, time=0.5547, value=0.0334}
          ]}}}))
    )
    Write to Module Namespace=true
   )
   &FN_Sprite_Attributes(
    Color=&ColorFromCurve(ColorCurve=DataInterface() {propertyValues {
       AlphaCurve {keys [<at 0 2.5 RCIM_Cubic>,<at 1 0 RCIM_Cubic>]}
       BlueCurve {keys [<at 0 0.2619>]}
       GreenCurve {keys [<at 0 3.7479>]}
       RedCurve {keys [<at 0 5>]}
      }})
    Create Color Attribute=true
    Create SpriteSize Attribute=true
    SpriteSize=&Multiply_Vector2D(
     A=<59.1852 100>
     B=&Vector2DFromFloat(Value=&FloatFromCurve(FloatCurve=DataInterface() {propertyValues {Curve {keys [
           {arriveTangent=7.1084, interpMode=RCIM_Cubic, leaveTangent=7.1084, tangentMode=RCTM_User, time=0, value=0.375}
           {arriveTangent=-2.821, interpMode=RCIM_Cubic, leaveTangent=-2.821, tangentMode=RCTM_User, time=0.1121, value=0.942}
           {tangentMode=RCTM_User, time=0.4, value=0}
          ]}}}))
    )
    Write to Module Namespace=true
   ) {module=FN_Sprite_Attributes003}
   &FN_Sprite_Attributes(
    Color=&ColorFromCurve(ColorCurve=DataInterface() {propertyValues {
       AlphaCurve {keys [<at 0 2.5 RCIM_Cubic>,<at 1 0 RCIM_Cubic>]}
       BlueCurve {keys [<at 0 0.7457>,<at 0.1895 0.3107>]}
       GreenCurve {keys [<at 0 4.2056>,<at 0.1895 1.7523>]}
       RedCurve {keys [<at 0 12>,<at 0.1895 5>]}
      }})
    Create Color Attribute=true
    Create SpriteSize Attribute=true
    SpriteSize=&Vector2DFromFloat(Value=&FloatFromCurve(
      FloatCurve=DataInterface() {propertyValues {Curve {keys [
          {arriveTangent=7.1084, interpMode=RCIM_Cubic, leaveTangent=7.1084, tangentMode=RCTM_User, time=0, value=0.375}
          {arriveTangent=-2.821, interpMode=RCIM_Cubic, leaveTangent=-2.821, tangentMode=RCTM_User, time=0.526, value=1.0318}
          {tangentMode=RCTM_User, time=1, value=0}
         ]}}}
      Scale Curve=80
     ))
    Write to Module Namespace=true
   ) {module=FN_Sprite_Attributes002}
   &FN_Sprite_Attributes(
    Color=&MakeLinearColorFromVectorAndFloat(
     "Vector (RGB)"=&MakeVectorFromLinearColorRGB(Color=&ColorFromCurve(ColorCurve=DataInterface() {propertyValues {
         AlphaCurve {keys [<at 0 1 RCIM_Cubic>,<at 1 0 RCIM_Cubic>]}
         BlueCurve {keys [<at 0 0 RCIM_Cubic>,<at 0.4548 0.1579 RCIM_Cubic>]}
         GreenCurve {keys [<at 0 2.9438 RCIM_Cubic>,<at 0.4548 0.9641 RCIM_Cubic>]}
         RedCurve {keys [<at 0 5 RCIM_Cubic>,<at 0.4548 2 RCIM_Cubic>]}
        }}))
     "Float (Alpha)"=&FloatFromCurve(FloatCurve=DataInterface() {propertyValues {Curve {
         defaultValue=1
         keys [
          {arriveTangent=9.0725, interpMode=RCIM_Cubic, leaveTangent=9.0725, tangentMode=RCTM_User, time=0, value=0}
          {arriveTangent=-2.5249, interpMode=RCIM_Cubic, leaveTangent=-2.5249, tangentMode=RCTM_User, time=0.1408, value=0.9847}
          {arriveTangent=-0.0405, interpMode=RCIM_Cubic, leaveTangent=-0.0405, tangentMode=RCTM_User, time=1, value=0}
         ]
        }}})
    )
    Create Color Attribute=true
    Create SpriteSize Attribute=true
    SpriteSize=&Multiply_Vector2D_ByFloat(
     Float=&FloatFromCurve(FloatCurve=DataInterface() {propertyValues {Curve {
         defaultValue=0
         keys [
          {arriveTangent=3.4473, interpMode=RCIM_Cubic, leaveTangent=3.4473, tangentMode=RCTM_User, time=0, value=0.75}
          {interpMode=RCIM_Cubic, tangentMode=RCTM_User, time=1, value=2}
         ]
        }}})
     Vector2D=&Vector2DFromFloat(Value=450)
    )
    Write to Module Namespace=true
   ) {module=FN_Sprite_Attributes004}
   &FN_Sprite_Attributes(
    Color=&MakeLinearColorFromVectorAndFloat(
     "Vector (RGB)"=&MakeVectorFromLinearColorRGB(Color=&ColorFromCurve(ColorCurve=DataInterface() {propertyValues {
         AlphaCurve {keys [<at 0 1 RCIM_Cubic>,<at 1 0 RCIM_Cubic>]}
         BlueCurve {keys [<at 0 0 RCIM_Cubic>,<at 0.4548 0.1579 RCIM_Cubic>]}
         GreenCurve {keys [<at 0 2.9438 RCIM_Cubic>,<at 0.4548 0.9641 RCIM_Cubic>]}
         RedCurve {keys [<at 0 5 RCIM_Cubic>,<at 0.4548 2 RCIM_Cubic>]}
        }}))
     "Float (Alpha)"=&FloatFromCurve(FloatCurve=DataInterface() {propertyValues {Curve {
         defaultValue=1
         keys [
          {arriveTangent=9.0725, interpMode=RCIM_Cubic, leaveTangent=9.0725, tangentMode=RCTM_User, time=0, value=0}
          {arriveTangent=-2.5249, interpMode=RCIM_Cubic, leaveTangent=-2.5249, tangentMode=RCTM_User, time=0.1408, value=0.9847}
          {arriveTangent=-0.0405, interpMode=RCIM_Cubic, leaveTangent=-0.0405, tangentMode=RCTM_User, time=1, value=0}
         ]
        }}})
    )
    Create Color Attribute=true
    Create SpriteSize Attribute=true
    SpriteSize=&Multiply_Vector2D_ByFloat(
     Float=&FloatFromCurve(FloatCurve=DataInterface() {propertyValues {Curve {
         defaultValue=0
         keys [
          {arriveTangent=3.4473, interpMode=RCIM_Cubic, leaveTangent=3.4473, tangentMode=RCTM_User, time=0, value=0.75}
          {interpMode=RCIM_Cubic, tangentMode=RCTM_User, time=1, value=2}
         ]
        }}})
     Vector2D=&Vector2DFromFloat(Value=150)
    )
    Write to Module Namespace=true
   ) {module=FN_Sprite_Attributes005}
   &FN_Sprite_Attributes(
    Camera Offset:NiagaraFloat=&FN_FrontEndFloatModifier(Gameplay=15, Front End=15)
    Color=&MakeLinearColorFromVectorAndFloat(
     "Vector (RGB)"=&MakeVectorFromLinearColorRGB(Color=<rgba 3 2.235 1.47 1>)
     "Float (Alpha)"=&Multiply_Float(
      A=&FloatFromCurve(FloatCurve=DataInterface() {propertyValues {Curve {keys [
           {arriveTangent=9.0725, interpMode=RCIM_Cubic, leaveTangent=9.0725, tangentMode=RCTM_User, time=0, value=0}
           {arriveTangent=-2.5249, interpMode=RCIM_Cubic, leaveTangent=-2.5249, tangentMode=RCTM_User, time=0.1408, value=0.9847}
           {arriveTangent=-0.0405, interpMode=RCIM_Cubic, leaveTangent=-0.0405, tangentMode=RCTM_User, time=1, value=0}
          ]}}})
      B=&FNTimeOfDayScalar(Night Scale=0.3, Morning Scale=0.5, Evening Scale=0.5)
     )
    )
    Create Camera Offset Attribute=true
    Create Color Attribute=true
    Create SpriteSize Attribute=true
    SpriteSize=&Vector2DFromFloat(Value=&FloatFromCurve(
      FloatCurve=DataInterface() {propertyValues {Curve {keys [{arriveTangent=0.8072, interpMode=RCIM_Cubic, leaveTangent=0.8072, tangentMode=RCTM_User, time=0, value=0.7}, <at 1 1.075>]}}}
      CurveIndex=&Multiply_Float(A=$Particles.NormalizedAge, B=1)
      Scale Curve=90
     ))
    Write to Module Namespace=true
   ) {module=FN_Sprite_Attributes001}
   /SetVariables(Particles.Dark_Glow=&MakeLinearColorFromVectorAndFloat(
     "Vector (RGB)"=&MakeVectorFromLinearColorRGB(Color=&MultiplyLinearColor(
       Linear Color=&ColorFromCurve(ColorCurve=DataInterface() {propertyValues {
          AlphaCurve {keys []}
          BlueCurve {keys [<at 0.2119 0.3281 RCIM_Cubic>,<at 0.5555 0.1406>,<at 0.9372 0.05 RCIM_Cubic>]}
          GreenCurve {keys [<at 0.2119 0.3281 RCIM_Cubic>,<at 0.5555 0.1406>,<at 0.9372 0.05 RCIM_Cubic>]}
          RedCurve {keys [<at 0.2119 0.3281 RCIM_Cubic>,<at 0.5555 0.1406>,<at 0.9372 0.05 RCIM_Cubic>]}
         }})
       Scale Factor=&MakeVector4_FromLinearColor(LinearColor=<rgba 0.1771 0.0985 0.0258 1>)
      ))
     "Float (Alpha)"=&Multiply_Float(
      A=0.4
      B=&FloatFromCurve(FloatCurve=DataInterface() {propertyValues {Curve {keys [
           {arriveTangent=-0.3275, interpMode=RCIM_Cubic, leaveTangent=-0.3275, tangentMode=RCTM_User, time=0, value=1}
           {arriveTangent=-0.3009, interpMode=RCIM_Cubic, leaveTangent=-0.3009, tangentMode=RCTM_User, time=1, value=0}
          ]}}})
     )
    )) {isSetParameters=true}
   /SetVariables(Particles.Dark_Glow_Size=&Multiply_Vector2(
     A=&Vector2DFromFloat(Value=135)
     B=&Vector2DFromFloat(Value=&FloatFromCurve(FloatCurve=DataInterface() {propertyValues {Curve {keys [
           {arriveTangent=4.1378, interpMode=RCIM_Cubic, leaveTangent=4.1378, tangentMode=RCTM_User, time=0, value=0}
           {interpMode=RCIM_Cubic, leaveTangent=0.5, time=1, value=1.45}
          ]}}}))
    )) {isSetParameters=true}
  ]
  rendererClasses [&NiagaraSpriteRendererProperties, &NiagaraLightRendererProperties]
  renderers [
   {rendererClass=&NiagaraSpriteRendererProperties}
   {rendererClass=&NiagaraSpriteRendererProperties, rendererIndex=1}
   {rendererClass=&NiagaraSpriteRendererProperties, rendererIndex=2}
   {rendererClass=&NiagaraSpriteRendererProperties, rendererIndex=3}
   {rendererClass=&NiagaraLightRendererProperties, rendererIndex=4}
   {rendererClass=&NiagaraSpriteRendererProperties, rendererIndex=5}
   {rendererClass=&NiagaraSpriteRendererProperties, rendererIndex=6}
   {rendererClass=&NiagaraSpriteRendererProperties, rendererIndex=7}
   {rendererClass=&NiagaraSpriteRendererProperties, rendererIndex=8}
   {rendererClass=&NiagaraSpriteRendererProperties, rendererIndex=9}
   {rendererClass=&NiagaraSpriteRendererProperties, rendererIndex=10}
   {rendererClass=&NiagaraSpriteRendererProperties, rendererIndex=11}
  ]
 }
 Minimal {
  emitterUpdateScript [
   &EmitterState()
   &SpawnBurst_Instantaneous(Spawn Count=8, Use Spawn Probability=false, UseLoopCountLimit=false)
  ]
  particleSpawnScript [
   &InitializeParticle(
    Color Maximum=<rgba 3 0.4747 0 0.75>
    Color Minimum=<rgba 3 1.555 0 0.75>
    Color Mode=::NewEnumerator2
    Lifetime=0.2
    Sprite Size Max=<3 175>
    Sprite Size Min=<1 120>
    Sprite Size Mode=::NewEnumerator2
    UsePositionOffset=false
    Write Material Random=true
   )
   &ShapeLocation(MaskSpawnGroup=false, Override Seed=false, Sphere Radius=8, UseOwnerScale=true)
   &AddVelocity(
    Module.Use Falloff Distance=false
    Override Seed=false
    Velocity Mode=::NewEnumerator1
    Velocity Speed=&RandomRangeFloat(Minimum=500, Maximum=625)
   )
  ]
  particleUpdateScript [
   &ParticleState()
   &SolveForcesAndVelocity(Clamp Velocity=false, Limit Acceleration=false)
   &ScaleColor(
    Scale Alpha=&FloatFromCurve(FloatCurve=DataInterface() {propertyValues {Curve {
        defaultValue=1
        keys [
         {tangentMode=RCTM_User, time=0, value=1}
         {arriveTangent=-0.0405, interpMode=RCIM_Cubic, leaveTangent=-0.0405, tangentMode=RCTM_User, time=0.7009, value=1}
         {arriveTangent=-8.0089, interpMode=RCIM_Cubic, leaveTangent=-8.0089, tangentMode=RCTM_User, time=1, value=0}
        ]
       }}})
    ScaleA=true
    ScaleRGB=true
   )
  ]
  rendererClasses [&NiagaraSpriteRendererProperties]
  renderers [{rendererClass=&NiagaraSpriteRendererProperties}]
 }
 Sparks {
  emitterUpdateScript [
   &EmitterState()
   &SpawnBurst_Instantaneous(Spawn Count=24, Use Spawn Probability=false)
   &SpawnBurst_Instantaneous(Spawn Count=24, Spawn Group=1, Use Spawn Probability=true) {module=SpawnBurst_Instantaneous001}
  ]
  particleSpawnScript [
   &InitializeParticle(
    Color=<rgba 1 0.6 0.2 1>
    Lifetime Max=0.6
    Lifetime Min=0.3
    Lifetime Mode=::NewEnumerator1
    Sprite Size Max=<2 8>
    Sprite Size Min=<2 4>
    Sprite Size Mode=::NewEnumerator2
    UsePositionOffset=false
    Write Material Random=true
   )
   &ShapeLocation(MaskSpawnGroup=false, Override Seed=false, Sphere Radius=8, UseOwnerScale=true)
   &AddVelocity(
    Module.Use Falloff Distance=false
    Override Seed=false
    Velocity Mode=::NewEnumerator1
    Velocity Origin=$Particles.ShapeLocation.ShapePosition
    Velocity Speed=/Niagara/DynamicInputs/Helpers/DistributionCurveFloat(
     Minimum:NiagaraFloat=400
     Maximum:NiagaraFloat=550
     Distribution Curve:NiagaraDataInterfaceCurve=DataInterface() {propertyValues {Curve {
        defaultValue=0
        keys [<at 0.4 0 RCIM_Cubic>, {arriveTangent=3.311, interpMode=RCIM_Cubic, leaveTangent=3.311, tangentMode=RCTM_User, time=1, value=1}]
       }}}
     Curve Parameter Type:NiagaraInt32=0
     "Curve Parameter Type:":NiagaraInt32=0
     Curve Parameter:NiagaraFloat=?
     Evaluation Type:ENiagaraRandomnessEvaluation=::NewEnumerator0
     Recalculate Random Each Loop:NiagaraBool=true
     Randomness Mode:ENiagaraRandomnessMode=::NewEnumerator0
     Random Seed:NiagaraInt32=0
     Fixed Random Seed:NiagaraBool=false
    )
   )
   &CameraOffset(Camera Offset Amount=30)
  ]
  particleUpdateScript [
   &ParticleState()
   &ScaleColor(
    Scale RGB=&VectorFromFloat(Value=&FloatFromCurve(
      FloatCurve=DataInterface() {propertyValues {Curve {keys [{interpMode=RCIM_Cubic, tangentMode=RCTM_User, time=0, value=10}, <at 1 1>]}}}
      Scale Curve=0.3
     ))
    ScaleA=true
    ScaleRGB=true
   ) {module=ScaleColor001}
   &Drag(
    Drag=&Multiply_Float(
     A=&FloatFromCurve(
      FloatCurve=DataInterface() {propertyValues {Curve {keys [<at 0 0.5>, {arriveTangent=0.5, interpMode=RCIM_Cubic, leaveTangent=0.5, time=0.3, value=0.5}, <at 1 1 RCIM_Cubic>]}}}
      Scale Curve=&RandomRangeFloat(Minimum=3, Maximum=5)
     )
     B=1
    )
    Use Linear Drag=true
    Use Rotational Drag=false
   )
   /Niagara/Modules/Update/Forces/CurlNoiseForce(
    Noise Frequency=10
    Noise Strength=&FloatFromCurve(FloatCurve=DataInterface() {propertyValues {Curve {keys [<at 0 1>,<at 1 0>]}}}, Scale Curve=3000)
    Pan Noise=true
    Pan Noise Field=<0.1 0.1 0.5>
   ) {enabled=false}
   &GravityForce(Gravity=&RandomRangeVector(Minimum=<0 0 -850>, Maximum=<0 0 -980>)) {module=GravityForce001}
   &SolveForcesAndVelocity(Clamp Velocity=false, Limit Acceleration=false)
   &ScaleSpriteSizeBySpeed(Max Scale Factor=<1 12>, Min Scale Factor=<0.5 0.1>, Velocity Threshold=800)
   &ScaleSpriteSize(Scale Factor=&MakeVector2D(
     X=&FloatFromCurve(FloatCurve=DataInterface() {propertyValues {Curve {keys [<at 0 0.5>,<at 0.1 1>,<at 1 0.5>]}}})
     Y=&FloatFromCurve(FloatCurve=DataInterface() {propertyValues {Curve {keys [<at 0 1>,<at 1 0>]}}})
    ))
   &ScaleSpriteSize(
    Scale Sprite Size Mode=::NewEnumerator0
    Uniform Scale Factor=&FN_FrontEndFloatModifier(Gameplay=1, Front End=$User.FrontEndScale)
   ) {enabled=false, module=ScaleSpriteSize001}
   &FN_Sprite_Attributes(
    Color=&MakeLinearColorFromVectorAndFloat(
     "Vector (RGB)"=&MakeVectorFromLinearColorRGB(Color=&MergeHelper_ColorSwitch(
       0=&ColorFromCurve(ColorCurve=DataInterface() {propertyValues {
          AlphaCurve {keys [<at 0 1 RCIM_Cubic>,<at 1 0 RCIM_Cubic>]}
          BlueCurve {keys [<at 0 0.14 RCIM_Cubic>,<at 0.9637 0.035 RCIM_Cubic>]}
          GreenCurve {keys [<at 0 0.512 RCIM_Cubic>,<at 0.9637 0.1125 RCIM_Cubic>]}
          RedCurve {keys [<at 0 2 RCIM_Cubic>,<at 0.9637 0.5 RCIM_Cubic>]}
         }})
       1=&ColorFromCurve(ColorCurve=DataInterface() {propertyValues {
          AlphaCurve {keys [<at 0 1 RCIM_Cubic>,<at 1 0 RCIM_Cubic>]}
          BlueCurve {keys [<at 0 2 RCIM_Cubic>,<at 0.9637 0.5 RCIM_Cubic>]}
          GreenCurve {keys [<at 0 1.38 RCIM_Cubic>,<at 0.9637 0.19 RCIM_Cubic>]}
          RedCurve {keys [<at 0 0.14 RCIM_Cubic>,<at 0.9637 0.035 RCIM_Cubic>]}
         }})
       2=?
       3=?
       Index=$System.Variant
       Num=::NewEnumerator0
      ))
     "Float (Alpha)"=&FloatFromCurve(FloatCurve=DataInterface() {propertyValues {Curve {keys [<at 0 1>,<at 0.8 0>]}}}, Scale Curve=0.25)
    )
    Create Color Attribute=true
    Create SpriteSize Attribute=true
    SpriteSize=&Multiply_Vector2D_ByFloat(
     Float=2
     Vector2D=&Multiply_Vector2D(
      A=<5 15>
      B=&MakeVector2D(
       X=&FloatFromCurve(FloatCurve=DataInterface() {propertyValues {Curve {keys [<at 0 0.5>,<at 0.1 1>,<at 1 0.5>]}}})
       Y=&FloatFromCurve(FloatCurve=DataInterface() {propertyValues {Curve {keys [<at 0 1>,<at 1 0>]}}})
      )
     )
    )
    Write to Module Namespace=true
   )
  ]
  rendererClasses [&NiagaraSpriteRendererProperties]
  renderers [
   {rendererClass=&NiagaraSpriteRendererProperties}
   {rendererClass=&NiagaraSpriteRendererProperties, rendererIndex=1}
  ]
 }
 Streaks {
  emitterUpdateScript [
   &EmitterState()
   &SpawnBurst_Instantaneous(Spawn Count=8, Use Spawn Probability=true, UseLoopCountLimit=false)
   &SpawnRate(SpawnRate=40, Use Spawn Probability=false) {enabled=false}
  ]
  particleSpawnScript [
   &InitializeParticle(
    Color=<rgba 2 0.9903 0.1574 1>
    Lifetime=0.2
    Sprite Rotation Mode=::NewEnumerator2
    Sprite Size Mode=::NewEnumerator3
    Sprite UV Mode=::NewEnumerator1
    Uniform Sprite Size=35
    UsePositionOffset=false
    Write Material Random=true
   )
   &ShapeLocation(
    MaskSpawnGroup=false
    Override Seed=false
    Rotation Mode=::NewEnumerator2
    Sphere Radius=6
    UseOwnerScale=true
    Yaw / Pitch / Roll=<0 0.25 0.25>
   )
   &AddVelocity(Module.Use Falloff Distance=false, Override Seed=false, Velocity Mode=::NewEnumerator1) {module=AddVelocity001}
  ]
  particleUpdateScript [
   &ParticleState()
   &ScaleSpriteSize(
    Scale Factor=&RandomRangeVector2D(Minimum=<1.75 2.5>, Maximum=<2.25 3.5>)
    Scale Sprite Size Mode=::NewEnumerator2
   )
   &ScaleSpriteSize(
    Non-Uniform Curve Scale=<0.3 1.7>
    Non-Uniform Curve Sprite Scale=DataInterface() {propertyValues {
      XCurve {keys [
        {arriveTangent=1.4948, interpMode=RCIM_Cubic, leaveTangent=1.4948, tangentMode=RCTM_User, time=0, value=1}
        {arriveTangent=-1.1904, interpMode=RCIM_Cubic, leaveTangent=-1.1904, tangentMode=RCTM_User, time=0.3, value=0.7}
        {arriveTangent=0.0224, interpMode=RCIM_Cubic, leaveTangent=0.0224, tangentMode=RCTM_User, time=1, value=0.5}
       ]}
      YCurve {keys [
        {arriveTangent=1.3782, interpMode=RCIM_Cubic, leaveTangent=1.3782, tangentMode=RCTM_User, time=0, value=0.4}
        {arriveTangent=0.9134, interpMode=RCIM_Cubic, leaveTangent=0.9134, tangentMode=RCTM_User, time=0.3, value=0.75}
        {arriveTangent=0.0224, interpMode=RCIM_Cubic, leaveTangent=0.0224, tangentMode=RCTM_User, time=1, value=1}
       ]}
     }}
    Scale Sprite Size Mode=::NewEnumerator3
   ) {module=ScaleSpriteSize001}
   &SolveForcesAndVelocity(Clamp Velocity=false, Limit Acceleration=false)
   &ScaleColor(
    Scale Alpha=&Multiply_Float(
     A=&FloatFromCurve(
      FloatCurve=DataInterface() {propertyValues {Curve {keys [
          {interpMode=RCIM_Cubic, tangentMode=RCTM_User, time=0, value=1}
          {arriveTangent=-0.4904, interpMode=RCIM_Cubic, leaveTangent=-0.4904, tangentMode=RCTM_User, time=1, value=0}
         ]}}}
      Scale Curve=&RandomRangeFloat(Minimum=2, Maximum=5)
     )
     B=0.1
    )
    Scale RGB=&VectorFromFloat(Value=&Lerp_Float(
      A=1
      B=3
      Alpha=&FloatFromCurve(FloatCurve=DataInterface() {propertyValues {Curve {keys [
           {arriveTangent=9.0725, interpMode=RCIM_Cubic, leaveTangent=9.0725, tangentMode=RCTM_User, time=0, value=0}
           {arriveTangent=0.0771, interpMode=RCIM_Cubic, leaveTangent=0.0771, tangentMode=RCTM_User, time=0.2, value=1}
           {arriveTangent=-0.0405, interpMode=RCIM_Cubic, leaveTangent=-0.0405, tangentMode=RCTM_User, time=1, value=0}
          ]}}})
     ))
    ScaleA=true
    ScaleRGB=true
   )
   &DynamicMaterialParameters(
    Index 0 Param 1=&FloatFromCurve(FloatCurve=DataInterface() {propertyValues {Curve {keys [
         {arriveTangent=0.2343, interpMode=RCIM_Cubic, leaveTangent=0.2343, tangentMode=RCTM_User, time=0.5, value=0}
         {arriveTangent=0.075, interpMode=RCIM_Cubic, leaveTangent=0.075, tangentMode=RCTM_User, time=1, value=1}
        ]}}})
    Index 0 Param 2=1.1
    Index 0 Param 4=0.5
    Module.Index0Param0Write=true
    Module.Index0Param1Write=true
    Module.Index0Param2Write=true
    Module.Index0Param3Write=true
   )
   &ScaleColor(
    Scale Alpha=&Multiply_Float(A=1, B=&FNTimeOfDayScalar(Night Scale=0.4, Morning Scale=0.6, Evening Scale=0.6))
    Scale RGB=&Multiply_VectorByFloat(
     Vector=<1 1 1>
     Float=&FloatFromCurve(FloatCurve=DataInterface() {propertyValues {Curve {keys [{arriveTangent=-23.3037, interpMode=RCIM_Cubic, leaveTangent=-23.3037, tangentMode=RCTM_User, time=0, value=40}, <at 1 1>]}}})
    )
    ScaleA=true
    ScaleRGB=true
   ) {module=ScaleColor001}
  ]
  rendererClasses [&NiagaraSpriteRendererProperties]
  renderers [
   {rendererClass=&NiagaraSpriteRendererProperties}
   {rendererClass=&NiagaraSpriteRendererProperties, rendererIndex=1}
  ]
 }
}
summary {
 systemName=NS_Emote_OathBound_ImpactBig
 emitters [
  {
   emitterName=Flash_Elements
   rendererClasses [&NiagaraSpriteRendererProperties, &NiagaraLightRendererProperties]
  }
  {emitterName=Sparks, rendererClasses [&NiagaraSpriteRendererProperties]}
  {emitterName=Streaks, rendererClasses [&NiagaraSpriteRendererProperties]}
  {emitterName=Minimal, rendererClasses [&NiagaraSpriteRendererProperties]}
 ]
 userVariables [
  User.bIsFrontEnd {type=NiagaraBool}
  User.bIsFrontEndPreview {type=NiagaraBool}
  User.FrontEndScale {type=NiagaraFloat, defaultValue=0.6, description=""}
  User.PhysicalSurface {type=NiagaraInt32}
  User.SwingSectionIndex {type=NiagaraInt32}
  User.VariantID {type=NiagaraFloat, defaultValue=0, description=""}
 ]
}
