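# Niagara effect for the OathBound emote that draws golden animation trails and a mesh glow on the animated character.
# Two FN_AnimTrail emitters spawn ribbons between bone or socket pairs, differing in color and width mode, and fade and taper over particle life.
# Curves indexed by emitter looped age drive trail lifetime, cutting and restoring trails through the emote.
# MainGlow bursts long-lived mesh particles that track the skeletal mesh and fade near the end.
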NiagaraSystem /BRCosmetics/Animation/Game/MainPlayer/Emotes/OathBound/FX/NS_Emote_OathBound_Trail
dependencies {
 usedDataInterfaces [
  &NiagaraDataInterfaceCurve
  &NiagaraDataInterfaceSkeletalMesh
  &NiagaraDataInterfaceSocketReader
 ]
 usedDynamicInputs []
 usedModules [
  &self:SystemSpawnScript.SystemScriptSource.SystemScriptGraph.NiagaraNodeAssignment_2.SetVariables_#1
  &SystemState
 ]
}
emitters {
 FN_AnimTrail {
  emitterSpawnScript [&FN_RibbonIDOffsetUtility()]
  emitterUpdateScript [
   &EmitterState()
   &FN_SpawnAnimTrail()
  ]
  particleSpawnScript [
   &FN_InitializeAnimTrailBoneIndecies()
   &FN_InitializeAnimTrailBoneIndecies(AnimTrail Selection=1, First Bone or Socket Index=2, Second Bone or Socket Index=3) {enabled=false, module=FN_InitializeAnimTrailBoneIndecies001}
   &FN_InitializeAnimTrail(
    Color=<rgba 1 0.5155 0.0462 0.8>
    Lifetime=&FloatFromCurve(
     FloatCurve=DataInterface() {propertyValues {Curve {keys [
         <at 0 1>
         <at 1.6 0.9805>
         <at 1.65 0>
         {arriveTangent=1.6667, interpMode=RCIM_Cubic, leaveTangent=1.6667, time=1.95, value=0}
         <at 2.1 0.75 RCIM_Cubic>
         {arriveTangent=-0.0174, interpMode=RCIM_Cubic, leaveTangent=-0.0174, tangentMode=RCTM_User, time=2.5, value=0}
         <at 2.85 0>
         <at 2.9 1.2>
         <at 3.15 1.2>
         <at 3.45 1>
         <at 3.7 1>
         <at 3.75 0.75>
         {arriveTangent=0.2778, interpMode=RCIM_Cubic, leaveTangent=0.2778, time=4.6, value=0.75}
         <at 4.65 1 RCIM_Cubic>
         {arriveTangent=-0.1841, interpMode=RCIM_Cubic, leaveTangent=-0.1841, time=5.35, value=0.75}
         {arriveTangent=-0.0357, interpMode=RCIM_Cubic, leaveTangent=-0.0357, time=6.0081, value=0.75}
         <at 6.75 0.7 RCIM_Cubic>
        ]}}}
     CurveIndex=$Emitter.LoopedAge
     Scale Curve=0.1
    )
   )
   &AddVelocity(Velocity=<0 400 0>) {enabled=false}
  ]
  particleUpdateScript [
   &SolveForcesAndVelocity(Clamp Velocity=false, Limit Acceleration=false)
   &FN_UpdateAnimTrail() {module=FN_UpdateAnimTrail001}
   &ScaleColor(
    Scale Alpha=&Multiply_Float(
     A=&FloatFromCurve(FloatCurve=DataInterface() {propertyValues {Curve {keys [<at 0 1>,<at 4.64 1>,<at 5.36 1>]}}}, CurveIndex=$Emitter.LoopedAge)
     B=&FloatFromCurve(FloatCurve=DataInterface() {propertyValues {Curve {
         defaultValue=1
         keys [{interpMode=RCIM_Cubic, tangentMode=RCTM_User, time=0, value=1}, {interpMode=RCIM_Cubic, tangentMode=RCTM_User, time=1, value=0}]
        }}})
    )
    ScaleA=true
    ScaleRGB=true
   )
   &ScaleRibbonWidth(Ribbon Width Scale=&FloatFromCurve(FloatCurve=DataInterface() {propertyValues {Curve {
        defaultValue=1
        keys [
         {tangentMode=RCTM_User, time=0, value=1}
         {arriveTangent=-0.0405, interpMode=RCIM_Cubic, leaveTangent=-0.0405, tangentMode=RCTM_User, time=0.7009, value=1}
         {arriveTangent=-8.0089, interpMode=RCIM_Cubic, leaveTangent=-8.0089, tangentMode=RCTM_User, time=1, value=0}
        ]
       }}}))
  ]
  rendererClasses [&NiagaraRibbonRendererProperties]
  renderers [
   {rendererClass=&NiagaraRibbonRendererProperties}
   {rendererClass=&NiagaraRibbonRendererProperties, rendererIndex=1}
  ]
 }
 FN_AnimTrail001 {
  emitterSpawnScript [&FN_RibbonIDOffsetUtility()]
  emitterUpdateScript [
   &EmitterState()
   &FN_SpawnAnimTrail(AnimTrailWidthMode=::FromSecond)
  ]
  particleSpawnScript [
   &FN_InitializeAnimTrailBoneIndecies()
   &FN_InitializeAnimTrailBoneIndecies(AnimTrail Selection=1, First Bone or Socket Index=2, Second Bone or Socket Index=3) {enabled=false, module=FN_InitializeAnimTrailBoneIndecies001}
   &FN_InitializeAnimTrail(
    Color=<rgba 1 0.7707 0 1.75>
    Lifetime=&FloatFromCurve(
     FloatCurve=DataInterface() {propertyValues {Curve {keys [
         <at 0 1>
         <at 1.6 0.9805>
         <at 1.65 0>
         {arriveTangent=1.6667, interpMode=RCIM_Cubic, leaveTangent=1.6667, time=1.95, value=0}
         <at 2.1 0.75 RCIM_Cubic>
         {arriveTangent=-0.0174, interpMode=RCIM_Cubic, leaveTangent=-0.0174, tangentMode=RCTM_User, time=2.5, value=0}
         <at 2.85 0>
         <at 2.9 1.2>
         <at 3.15 1.2>
         <at 3.45 1>
         <at 3.7 1>
         <at 3.75 0.75>
         {arriveTangent=0.2778, interpMode=RCIM_Cubic, leaveTangent=0.2778, time=4.6, value=0.75}
         <at 4.65 1 RCIM_Cubic>
         {arriveTangent=-0.1841, interpMode=RCIM_Cubic, leaveTangent=-0.1841, time=5.35, value=0.75}
         {arriveTangent=-0.0357, interpMode=RCIM_Cubic, leaveTangent=-0.0357, time=6.0081, value=0.75}
         <at 6.75 0.7 RCIM_Cubic>
        ]}}}
     CurveIndex=$Emitter.LoopedAge
     Scale Curve=0.1
    )
   )
   &AddVelocity(Velocity=<0 400 0>) {enabled=false}
  ]
  particleUpdateScript [
   &SolveForcesAndVelocity(Clamp Velocity=false, Limit Acceleration=false)
   &FN_UpdateAnimTrail() {module=FN_UpdateAnimTrail001}
   &ScaleColor(
    Scale Alpha=&FloatFromCurve(FloatCurve=DataInterface() {propertyValues {Curve {
        defaultValue=1
        keys [{interpMode=RCIM_Cubic, tangentMode=RCTM_User, time=0, value=1}, {interpMode=RCIM_Cubic, tangentMode=RCTM_User, time=1, value=0}]
       }}})
    ScaleA=true
    ScaleRGB=true
   )
   &ScaleRibbonWidth(Ribbon Width Scale=&FloatFromCurve(FloatCurve=DataInterface() {propertyValues {Curve {
        defaultValue=1
        keys [
         {tangentMode=RCTM_User, time=0, value=1}
         {arriveTangent=-0.0405, interpMode=RCIM_Cubic, leaveTangent=-0.0405, tangentMode=RCTM_User, time=0.7009, value=1}
         {arriveTangent=-8.0089, interpMode=RCIM_Cubic, leaveTangent=-8.0089, tangentMode=RCTM_User, time=1, value=0}
        ]
       }}}))
  ]
  rendererClasses [&NiagaraRibbonRendererProperties]
  renderers [
   {rendererClass=&NiagaraRibbonRendererProperties}
   {rendererClass=&NiagaraRibbonRendererProperties, rendererIndex=1}
  ]
 }
 MainGlow {
  emitterUpdateScript [
   &EmitterState()
   &SpawnBurst_Instantaneous(Use Spawn Probability=false, UseLoopCountLimit=false)
  ]
  particleSpawnScript [&InitializeParticle(Color=<rgba 5 3.145 0.599 1>, Lifetime=9.5, UsePositionOffset=false, Write Material Random=true)]
  particleUpdateScript [
   &ParticleState()
   &SkeletalMeshLocation(Bone Sampling Mode=::NewEnumerator0, MaskSpawnGroup=false, OverrideRandomSeed=false, Skeletal Mesh=$System.Skeletal Mesh Reader)
   &ScaleColor(
    Scale Alpha=&FloatFromCurve(FloatCurve=DataInterface() {propertyValues {Curve {keys [<at 0 1>,<at 0.9 1>,<at 0.95 0>]}}})
    ScaleA=true
    ScaleRGB=true
   )
  ]
  rendererClasses [&NiagaraMeshRendererProperties]
  renderers [
   {rendererClass=&NiagaraMeshRendererProperties, rendererIndex=0}
   {rendererClass=&NiagaraMeshRendererProperties, rendererIndex=1}
  ]
 }
}
summary {
 systemName=NS_Emote_OathBound_Trail
 emitters [
  {emitterName=MainGlow, rendererClasses [&NiagaraMeshRendererProperties]}
  {emitterName=FN_AnimTrail, rendererClasses [&NiagaraRibbonRendererProperties]}
  {emitterName=FN_AnimTrail001, rendererClasses [&NiagaraRibbonRendererProperties]}
 ]
 userVariables [
  User.ParticleSpawnRate {type=NiagaraFloat, defaultValue=0, description=""}
  User.RibbonIDOffset {type=NiagaraInt32}
 ]
}
